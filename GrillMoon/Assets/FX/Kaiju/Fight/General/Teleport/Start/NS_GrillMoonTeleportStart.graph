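# Niagara effect for the start of a GrillMoon kaiju teleport, coloring several layers red-pink or blue-violet from the WanderingBoss user flag.
# Active emitters run on staggered loop delays: a pelvis-attached twist-absorb sprite with a dark companion, a portal of three scaled meshes plus a sprite, smoke, and bone-attached chromatic refraction sprites.
# Sizes mostly scale by a system FX Scale, and curve-driven dynamic material parameters animate the layers.
# BodyDecals and ribbon Trail emitters are disabled.
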NiagaraSystem /GrillMoon/Assets/FX/Kaiju/Fight/General/Teleport/Start/NS_GrillMoonTeleportStart
dependencies {
 usedDataInterfaces [
  &NiagaraDataInterfaceCurve
  &NiagaraDataInterfaceSkeletalMesh
  &NiagaraDataInterfaceColorCurve
  &NiagaraDataInterfaceArrayFloat3
  &NiagaraDataInterfaceArrayFloat
  &NiagaraDataInterfaceVectorCurve
 ]
 usedDynamicInputs [&RandomRangeLinearColor]
 usedModules [
  &self:SystemSpawnScript.SystemScriptSource.SystemScriptGraph.NiagaraNodeAssignment_0.SetVariables_#1
  &self:SystemSpawnScript.SystemScriptSource.SystemScriptGraph.NiagaraNodeAssignment_1.SetVariables_#2
  &self:SystemSpawnScript.SystemScriptSource.SystemScriptGraph.NiagaraNodeAssignment_2.SetVariables_#3
  &self:SystemSpawnScript.SystemScriptSource.SystemScriptGraph.NiagaraNodeAssignment_3.SetVariables_#4
  &SystemState
 ]
}
emitters {
 BodyDecals {
  bEnabled=false
  emitterUpdateScript [
   &EmitterState()
   &SpawnBurst_Instantaneous(Spawn Count=9, Use Spawn Probability=false, UseLoopCountLimit=false)
  ]
  particleSpawnScript [
   &PartitionParticles(Number Of Partitions=9)
   &InitializeParticle(
    Color=&MakeCustomLinearColorFromBool(Bool=$User.WanderingBoss, True Color=<rgba 3 0 0.25 1>, False Color=<rgba 0.5958 0 3 1>)
    Lifetime=0.5
    Sprite Size Mode=::NewEnumerator3
    Uniform Sprite Size=&Multiply_Float(A=30, B=$System.FX Scale)
    UsePositionOffset=false
    Write Material Random=true
   )
   &CameraOffset(Camera Offset Amount=&Multiply_Float(A=10, B=$System.FX Scale))
  ]
  particleUpdateScript [
   &ParticleState()
   &SkeletalMeshLocation(
    Bone / Socket Index=$Particles.PartitionParticles.Partition
    Bone Sampling Mode=::NewEnumerator4
    MaskSpawnGroup=false
    Orientation Sampling=::NewEnumerator1
    Skeletal Mesh=DataInterface() {propertyValues {
      DefaultMesh=null
      ExcludeBoneName=root
      OptionalComponentClass=null
      PreviewMesh=/Game/Characters/Player/Male/Medium/Base/SK_M_MALE_Base_Skeleton
      SkinningMode=SkinOnTheFly
      SoftSourceActor=null
      SourceMode=Default
      UvSetIndex=0
      WholeMeshLOD=-1
      bExcludeBone=false
      bReadDeformedGeometry=true
      bRequireCurrentFrameData=true
      ComponentTags []
      FilteredBones [spine_05,neck_01,spine_02,pelvis,thigh_r,thigh_l,calf_r,calf_l,head]
      FilteredSockets []
      MeshUserParameter {parameter {name=None, typeDefHandle {registeredTypeIndex=1}, varData []}}
      SamplingRegions []
     }}
   )
   /Niagara/Modules/Update/Utility/IncrementOverTime(Rate of Change:NiagaraFloat=&Lerp_Float(A=-1.1, B=2, Alpha=$System.NormalizedVelocityRange))
   &DynamicMaterialParameters(
    Index 0 Param 1=&FloatFromCurve(
     FloatCurve=DataInterface() {propertyValues {Curve {keys [{arriveTangent=3.2804, interpMode=RCIM_Cubic, leaveTangent=3.2804, tangentMode=RCTM_User, time=0, value=1}, <at 0.7 2>]}}}
     CurveIndex=$System.Age
    )
    Index 0 Param 2=1
    Index 0 Param 3=&Add_Float(
     A=$Particles.IncrementOverTime.Counter
     B=&FloatFromCurve(
      FloatCurve=DataInterface() {propertyValues {Curve {keys [{arriveTangent=-2.1645, interpMode=RCIM_Cubic, leaveTangent=-2.1645, tangentMode=RCTM_User, time=0, value=1}, <at 1.25 0>]}}}
      CurveIndex=$System.Age
      Scale Curve=-0.5
     )
    )
    Index 0 Param 4=&RandomRangeFloat(Minimum=0.75, Maximum=0.9)
    Index 1 Param 1=50
    Index 1 Param 2=10
    Module.Param0WriteEnabled=true
    Module.Param0WriteEnabled1=true
    Module.Param1WriteEnabled=true
    Module.Param1WriteEnabled1=true
    Module.Param2WriteEnabled=true
    Module.Param2WriteEnabled1=false
    Module.Param3WriteEnabled=true
    Module.Param3WriteEnabled1=false
    Write Parameter Index 1=true
   )
   &ScaleColor(
    Scale RGB=&VectorFromFloat(Value=&Lerp_Float(
      A=1
      B=6
      Alpha=&FloatFromCurve(FloatCurve=DataInterface() {propertyValues {Curve {keys [
           {interpMode=RCIM_Cubic, tangentMode=RCTM_User, time=0, value=0}
           {arriveTangent=2.1597, interpMode=RCIM_Cubic, leaveTangent=2.1597, tangentMode=RCTM_User, time=0.75, value=1}
           {tangentMode=RCTM_User, time=1, value=0}
          ]}}})
     ))
    ScaleA=true
    ScaleRGB=true
   )
  ]
  rendererClasses [&NiagaraSpriteRendererProperties]
  renderers [{rendererClass=&NiagaraSpriteRendererProperties}]
 }
 Copied_ChromaticRefraction {
  emitterUpdateScript [
   &EmitterState(
    Life Cycle Mode=::NewEnumerator1
    Loop Behavior=::NewEnumerator1
    Loop Delay=$System.Delay
    Loop Duration Mode=::NewEnumerator1
    UseLoopDelay=true
   )
   &SpawnBurst_Instantaneous(Use Spawn Probability=false, UseLoopCountLimit=false)
   /SetVariables(Emitter.A_1:LinearColor=&MakeCustomLinearColorFromBool(Bool=$User.WanderingBoss, True Color=<rgba 1 0.078 0.0934 1>, False Color=<rgba 0.2879 0 1 1>)) {isSetParameters=true}
   /SetVariables(Emitter.A_2:LinearColor=&MakeCustomLinearColorFromBool(Bool=$User.WanderingBoss, True Color=<rgba 1 0.092 0.2585 1>, False Color=<rgba 0.3412 0 1 1>)) {isSetParameters=true}
   /SetVariables() {isSetParameters=true}
  ]
  particleSpawnScript [
   &PartitionParticles(Number Of Partitions=3)
   &InitializeParticle(
    Lifetime=0.5
    Mesh Scale Mode=::NewEnumerator3
    Sprite Size Mode=::NewEnumerator3
    Uniform Sprite Size=&Multiply_Float(A=10, B=$System.FX Scale)
    UsePositionOffset=false
    Write Material Random=true
   )
  ]
  particleUpdateScript [
   &ParticleState()
   &Timeline(
    Play Rate=&FloatFromCurve(
     FloatCurve=DataInterface() {propertyValues {Curve {keys [
         {interpMode=RCIM_Cubic, leaveTangent=-12.5334, tangentMode=RCTM_Break, time=0, value=4}
         {interpMode=RCIM_Cubic, tangentMode=RCTM_Break, time=0.7, value=1}
        ]}}}
     Scale Curve=0.5
    )
    Timeline End=1.0000000200408773e+20
   )
   &ScaleSpriteSize(
    Scale Sprite Size Mode=::NewEnumerator0
    Uniform Scale Factor=&FloatFromCurve(
     FloatCurve=DataInterface() {propertyValues {Curve {keys [
         {interpMode=RCIM_Cubic, tangentMode=RCTM_User, time=0, value=0}
         {interpMode=RCIM_Cubic, tangentMode=RCTM_User, time=0.5, value=1}
         {tangentMode=RCTM_User, time=1, value=0}
        ]}}}
     Scale Curve=70
    )
   )
   &SkeletalMeshLocation(
    Bone / Socket Index=&Add_Integer(A=$Particles.PartitionParticles.Partition, B=12)
    Bone Sampling Mode=::NewEnumerator4
    MaskSpawnGroup=false
    Sampled Position Offset=&SelectVectorFromArray(
     Direct Array Index=$Particles.PartitionParticles.Partition
     Vector Selection Array=DataInterface() {propertyValues {FloatData [<0 0 0>, <-6 0 0>, <6 0 0>]}}
    )
    Skeletal Mesh=$System.SK_Mesh
   )
   &DynamicMaterialParameters(
    Index 0 Param 1=2
    Index 0 Param 2=1
    Index 0 Param 3=2
    Index 0 Param 4=1
    Index 1 Param 2=500
    Index 1 Param 3=0.5
    Index 1 Param 4=1
    Index 2 Param 1=$Particles.Timeline.Time
    Index 2 Param 2=$Particles.Timeline.Time
    Index 2 Param 3=$Particles.Timeline.Time
    Index 2 Param 4=&SelectFloatFromArray(Direct Array Index=$Particles.PartitionParticles.Partition, Float Selection Array=DataInterface() {propertyValues {FloatData [1,0.3,0.3]}})
    Index 3 Param 1=&FloatFromCurve(FloatCurve=DataInterface() {propertyValues {Curve {keys [<at 0.1 1>,<at 0.4 0>]}}})
    Index 3 Param 2=200
    Module.Index0Param0Write=true
    Module.Index0Param1Write=true
    Module.Index0Param2Write=true
    Module.Index0Param3Write=true
    Module.Index1Param0Write=true
    Module.Index1Param1Write=true
    Module.Index1Param2Write=true
    Module.Index1Param3Write=true
    Module.Index2Param0Write=true
    Module.Index2Param1Write=true
    Module.Index2Param2Write=true
    Module.Index2Param3Write=true
    Module.Index3Param0Write=true
    Module.Index3Param1Write=true
    Module.Index3Param2Write=true
    Module.Index3Param3Write=true
    Write Parameter Index 1=true
    Write Parameter Index 2=true
    Write Parameter Index 3=true
   ) {module=DynamicMaterialParameters001}
   &ScaleColor(Scale RGB=&VectorFromFloat(Value=0.5), ScaleA=true, ScaleRGB=true) {module=ScaleColor001}
  ]
  rendererClasses [&NiagaraSpriteRendererProperties]
  renderers [{rendererClass=&NiagaraSpriteRendererProperties}]
 }
 ExitPortal {
  emitterUpdateScript [
   &EmitterState(Life Cycle Mode=::NewEnumerator1, Loop Behavior=::NewEnumerator1, Loop Delay=$System.Delay, Loop Duration=0.35, UseLoopDelay=true)
   &SpawnBurst_Instantaneous(Use Spawn Probability=false, UseLoopCountLimit=false)
   /SetVariables(Emitter.PortalHueShift:NiagaraFloat=&MakeCustomFloatFromBool(Bool=$User.WanderingBoss, True Float=0.2, False Float=0)) {isSetParameters=true}
  ]
  particleSpawnScript [
   /SetVariables(Particles.VisibilityTag=$Emitter.SpawnGroup) {enabled=false, isSetParameters=true}
   &InitializeParticle(Color Mode=::NewEnumerator0, Lifetime=0.32, UsePositionOffset=false, Write Material Random=true)
   &CameraOffset(Camera Offset Amount=300) {enabled=false}
  ]
  particleUpdateScript [
   &ParticleState()
   /SetVariables(
    Particles.CameraOffset:NiagaraFloat=&Lerp_Float(
     A=75
     B=-175
     Alpha=&FloatFromCurve(FloatCurve=DataInterface() {propertyValues {Curve {keys [
          {arriveTangent=0.156, interpMode=RCIM_Cubic, leaveTangent=0.156, tangentMode=RCTM_User, time=0.2, value=0}
          {arriveTangent=0.5245, interpMode=RCIM_Cubic, leaveTangent=0.5245, tangentMode=RCTM_User, time=0.8, value=1}
         ]}}})
    )
    Particles.CameraOfs_Outline:NiagaraFloat=&Lerp_Float(
     A=80
     B=-150
     Alpha=&FloatFromCurve(FloatCurve=DataInterface() {propertyValues {Curve {keys [
          {arriveTangent=0.247, interpMode=RCIM_Cubic, leaveTangent=0.247, tangentMode=RCTM_User, time=0.2, value=0}
          {arriveTangent=0.5245, interpMode=RCIM_Cubic, leaveTangent=0.5245, tangentMode=RCTM_User, time=0.8, value=1}
         ]}}})
    )
    Particles.Color=&ColorFromCurve(ColorCurve=DataInterface() {propertyValues {
       AlphaCurve {keys [<at 0 0 RCIM_Cubic>,<at 0.0826 2 RCIM_Cubic>]}
       BlueCurve {keys [<at 0.4494 0 RCIM_Cubic>]}
       GreenCurve {keys [<at 0.4494 0 RCIM_Cubic>]}
       RedCurve {keys [<at 0.4494 0 RCIM_Cubic>]}
      }})
    Particles.Color2=&ColorFromCurve(ColorCurve=DataInterface() {propertyValues {
       AlphaCurve {keys [<at 0 0 RCIM_Cubic>,<at 0.0826 3 RCIM_Cubic>]}
       BlueCurve {keys [<at 0.4494 0 RCIM_Cubic>]}
       GreenCurve {keys [<at 0.4494 0 RCIM_Cubic>]}
       RedCurve {keys [<at 0.4494 0 RCIM_Cubic>]}
      }})
    Particles.Color_Outline:LinearColor=&MakeCustomLinearColorFromBool(Bool=$User.WanderingBoss, True Color=<rgba 6 0.456 0.8256 1>, False Color=<rgba 1.2454 0 6 1>)
   ) {isSetParameters=true}
   /SetVariables(
    Particles.CoreCameraOfs:NiagaraFloat=&FloatFromCurve(
     FloatCurve=DataInterface() {propertyValues {Curve {keys [{arriveTangent=-2.2178, interpMode=RCIM_Cubic, leaveTangent=-2.2178, tangentMode=RCTM_User, time=0.1, value=1}, <at 0.6 -1>]}}}
     Scale Curve=50
    )
    Particles.DynamicParam0_inky:Vector4f=&MakeVector4(
     X=&FloatFromCurve(
      FloatCurve=DataInterface() {propertyValues {Curve {keys [
          {interpMode=RCIM_Cubic, tangentMode=RCTM_User, time=0, value=0}
          {interpMode=RCIM_Cubic, tangentMode=RCTM_User, time=0.4, value=1.1}
          {interpMode=RCIM_Cubic, leaveTangent=-1.8333, time=1, value=1}
         ]}}}
      Scale Curve=0.5
     )
     Y=&RandomRangeFloat(Minimum=0, Maximum=1)
     Z=&FloatFromCurve(FloatCurve=DataInterface() {propertyValues {Curve {keys [
          {arriveTangent=0.3291, interpMode=RCIM_Cubic, leaveTangent=0.3291, tangentMode=RCTM_User, time=0, value=0}
          {arriveTangent=1.4729, tangentMode=RCTM_User, time=1, value=1}
         ]}}})
     W=&FloatFromCurve(FloatCurve=DataInterface() {propertyValues {Curve {keys [<at 0.45 0 RCIM_Cubic>, {arriveTangent=57.34, tangentMode=RCTM_User, time=0.999, value=15}]}}})
    )
    Particles.DynamicParam3:Vector4f=&MakeVector4(
     X=&FloatFromCurve(
      FloatCurve=DataInterface() {propertyValues {Curve {keys [
          {arriveTangent=7.0396, interpMode=RCIM_Cubic, leaveTangent=7.0396, tangentMode=RCTM_User, time=0, value=0}
          {arriveTangent=1.9836, interpMode=RCIM_Cubic, leaveTangent=4.3833, tangentMode=RCTM_Break, time=0.1, value=0.75}
          {arriveTangent=-0.2155, interpMode=RCIM_Cubic, leaveTangent=-0.2155, tangentMode=RCTM_User, time=0.3, value=1.25}
          {arriveTangent=-2.1167, interpMode=RCIM_Cubic, leaveTangent=-1.2047, tangentMode=RCTM_Break, time=0.8, value=0.775}
          {arriveTangent=-2.386, interpMode=RCIM_Cubic, leaveTangent=-2.386, tangentMode=RCTM_User, time=1, value=0}
         ]}}}
      Scale Curve=&Multiply_Float(A=2, B=$System.FX Scale)
     )
     Y=&FloatFromCurve(
      FloatCurve=DataInterface() {propertyValues {Curve {keys [
          {arriveTangent=-1.1824, interpMode=RCIM_Cubic, leaveTangent=-1.1824, tangentMode=RCTM_User, time=0, value=0}
          {arriveTangent=3.1431, interpMode=RCIM_Cubic, leaveTangent=3.1431, tangentMode=RCTM_User, time=0.1, value=0.875}
          {arriveTangent=0.95, interpMode=RCIM_Cubic, leaveTangent=0.95, time=0.3, value=1.625}
          {arriveTangent=-2.3214, interpMode=RCIM_Cubic, leaveTangent=-2.3214, time=0.6, value=1.35}
          {tangentMode=RCTM_User, time=1, value=0}
         ]}}}
      Scale Curve=25
     )
     Z=0
     W=1
    )
    Particles.Scale=&Multiply_VectorByFloat(
     Vector=&Multiply_Vector(
      A=&VectorFromFloat(Value=&FloatFromCurve(
        FloatCurve=DataInterface() {propertyValues {Curve {keys [
            {arriveTangent=6.6826, interpMode=RCIM_Cubic, leaveTangent=6.6826, tangentMode=RCTM_User, time=0, value=0}
            {arriveTangent=1.914, interpMode=RCIM_Cubic, leaveTangent=0.736, tangentMode=RCTM_Break, time=0.0951, value=0.9887}
            {arriveTangent=-0.7631, interpMode=RCIM_Cubic, leaveTangent=-1.2047, tangentMode=RCTM_Break, time=0.864, value=0.9964}
            {arriveTangent=-5.5008, interpMode=RCIM_Cubic, leaveTangent=-5.5008, tangentMode=RCTM_User, time=1, value=0}
           ]}}}
        CurveIndex=$Emitter.NormalizedLoopAge
       ))
      B=<-0.88 1.12 0.32>
     )
     Float=$System.FX Scale
    )
    Particles.Scale2:Vector3f=&Multiply_VectorByFloat(
     Vector=&Multiply_Vector(
      A=&VectorFromFloat(Value=&FloatFromCurve(FloatCurve=DataInterface() {propertyValues {Curve {keys [
            {arriveTangent=6.6826, interpMode=RCIM_Cubic, leaveTangent=6.6826, tangentMode=RCTM_User, time=0, value=0}
            {arriveTangent=1.914, interpMode=RCIM_Cubic, leaveTangent=3.6448, tangentMode=RCTM_Break, time=0.1, value=0.975}
            {arriveTangent=0.6667, interpMode=RCIM_Cubic, leaveTangent=0.6667, time=0.3, value=1.425}
            {arriveTangent=-0.7599, interpMode=RCIM_Cubic, leaveTangent=-0.7599, time=0.7, value=1.375}
            {arriveTangent=-4.066, interpMode=RCIM_Cubic, leaveTangent=-1.2047, tangentMode=RCTM_Break, time=0.864, value=0.9964}
            {arriveTangent=-5.5008, interpMode=RCIM_Cubic, leaveTangent=-5.5008, tangentMode=RCTM_User, time=1, value=0}
           ]}}}))
      B=<0.72 0.72 0.32>
     )
     Float=$System.FX Scale
    )
    Particles.Scale3:Vector3f=&Multiply_VectorByFloat(
     Vector=&Multiply_Vector(
      A=&VectorFromFloat(Value=&FloatFromCurve(FloatCurve=DataInterface() {propertyValues {Curve {keys [
            {arriveTangent=6.6826, interpMode=RCIM_Cubic, leaveTangent=6.6826, tangentMode=RCTM_User, time=0, value=0}
            {arriveTangent=1.914, interpMode=RCIM_Cubic, leaveTangent=1.8477, tangentMode=RCTM_Break, time=0.1486, value=1.125}
            {arriveTangent=-0.1797, interpMode=RCIM_Cubic, leaveTangent=-0.1797, time=0.6, value=1.525}
            {arriveTangent=-5.165, interpMode=RCIM_Cubic, leaveTangent=-1.2047, tangentMode=RCTM_Break, time=0.864, value=0.9964}
            {arriveTangent=-5.5008, interpMode=RCIM_Cubic, leaveTangent=-5.5008, tangentMode=RCTM_User, time=1, value=0}
           ]}}}))
      B=<0.52 0.52 0.36>
     )
     Float=$System.FX Scale
    )
    Particles.SpriteSize=&Multiply_Vector2D_ByFloat(
     Float=&MakeFloatFromVector(
      Vector=&Multiply_Vector(
       A=&VectorFromFloat(Value=&FloatFromCurve(FloatCurve=DataInterface() {propertyValues {Curve {keys [
             {arriveTangent=1.914, interpMode=RCIM_Cubic, leaveTangent=-0.0338, tangentMode=RCTM_Break, time=0.1, value=1}
             {arriveTangent=0.1217, interpMode=RCIM_Cubic, leaveTangent=-1.2047, tangentMode=RCTM_Break, time=1, value=1.1}
            ]}}}))
       B=<1.4 1.4 0.4>
      )
      Channel=::NewEnumerator0
     )
     Vector2D=&Multiply_Vector2D_ByFloat(Float=$System.FX Scale, Vector2D=<320 320>)
    )
   ) {isSetParameters=true}
   &InheritSourceMovement(
    Applied Position Scale=&VectorFromFloat(Value=&FloatFromCurve(FloatCurve=DataInterface() {propertyValues {Curve {keys [{interpMode=RCIM_Cubic, tangentMode=RCTM_User, time=0, value=1}, {interpMode=RCIM_Cubic, tangentMode=RCTM_User, time=1, value=0}]}}}))
    ApplyForce=false
    ApplyPosition=true
    ApplyPositionFalloff=false
   )
   &SolveForcesAndVelocity(Clamp Velocity=false, Limit Acceleration=false)
  ]
  rendererClasses [&NiagaraMeshRendererProperties, &NiagaraSpriteRendererProperties]
  renderers [
   {rendererClass=&NiagaraMeshRendererProperties, rendererIndex=0}
   {rendererClass=&NiagaraMeshRendererProperties, rendererIndex=1}
   {rendererClass=&NiagaraMeshRendererProperties, rendererIndex=2}
   {rendererClass=&NiagaraSpriteRendererProperties, rendererIndex=3}
  ]
 }
 SmokeLoop {
  emitterUpdateScript [
   &EmitterState(Life Cycle Mode=::NewEnumerator1, Loop Behavior=::NewEnumerator1, Loop Delay=0.08, Loop Duration=0.4, UseLoopDelay=true)
   &SpawnRate(SpawnRate=20, Use Spawn Probability=false) {module=SpawnRate001}
   &SpawnBurst_Instantaneous(Spawn Count=8, Spawn Group=1, Use Spawn Probability=false, UseLoopCountLimit=false)
  ]
  particleSpawnScript [
   &InitializeParticle(
    Color=$System.EnergyColor
    Lifetime=&RandomRangeFloat(Minimum=0.8, Maximum=1.3)
    Sprite Rotation Mode=::NewEnumerator2
    Sprite Size=&Multiply_Vector2D_ByFloat(
     Float=$System.FX Scale
     Vector2D=&Vector2DFromFloat(Value=&RandomRangeFloat(Minimum=200, Maximum=250))
    )
    Sprite Size Mode=::NewEnumerator4
    Sprite UV Mode=::NewEnumerator3
    UsePositionOffset=false
    Write Material Random=true
   )
   &CameraOffset(Camera Offset Amount=100)
   &DynamicMaterialParameters(
    Index 0 Param 1=&RandomRangeFloat(Minimum=0, Maximum=-6)
    Module.Index0Param0Write=true
    Module.Index0Param1Write=false
    Module.Index0Param2Write=false
    Module.Index0Param3Write=false
   )
   &AddVelocity(Velocity=&RandomRangeVector(Minimum=<0 0 30>, Maximum=<0 0 -30>))
   &ShapeLocation(MaskSpawnGroup=false, Non Uniform Scale=<0.3 0.3 1>, Override Seed=false, UseOwnerScale=true)
  ]
  particleUpdateScript [
   &ParticleState()
   &ScaleSpriteSize(
    Scale Sprite Size Mode=::NewEnumerator0
    Uniform Scale Factor=&MaskFloatBySpawnGroup(
     Masked Float=&FloatFromCurve(
      FloatCurve=DataInterface() {propertyValues {Curve {keys [
          <at 0 0.1 RCIM_Cubic>
          {arriveTangent=0.9, interpMode=RCIM_Cubic, leaveTangent=0.9, time=0.4, value=0.4}
          {interpMode=RCIM_Cubic, leaveTangent=0.1472, time=1, value=1}
         ]}}}
      Scale Curve=0.7
     )
     Passthrough Float=&FloatFromCurve(FloatCurve=DataInterface() {propertyValues {Curve {keys [
          {arriveTangent=4.1785, interpMode=RCIM_Cubic, leaveTangent=4.1785, tangentMode=RCTM_User, time=0, value=0}
          {arriveTangent=1.1179, interpMode=RCIM_Cubic, leaveTangent=1.1179, tangentMode=RCTM_User, time=0.2, value=0.975}
          {arriveTangent=0.0224, interpMode=RCIM_Cubic, leaveTangent=0.0224, tangentMode=RCTM_User, time=1, value=1}
         ]}}})
    )
   )
   &ScaleColor(
    Scale Alpha=&MaskFloatBySpawnGroup(
     Masked Float=&FloatFromCurve(FloatCurve=DataInterface() {propertyValues {Curve {keys [
          <at 0 0 RCIM_Cubic>
          {arriveTangent=2.2188, interpMode=RCIM_Cubic, leaveTangent=2.2188, time=0.2, value=1}
          {arriveTangent=-1.25, interpMode=RCIM_Cubic, leaveTangent=-1.25, time=0.8, value=1.775}
          {interpMode=RCIM_Cubic, leaveTangent=-26.9591, time=1, value=0}
         ]}}})
     Passthrough Float=&FloatFromCurve(
      FloatCurve=DataInterface() {propertyValues {Curve {keys [
          {tangentMode=RCTM_User, time=0, value=1}
          {arriveTangent=-0.0405, interpMode=RCIM_Cubic, leaveTangent=-0.0405, tangentMode=RCTM_User, time=0.6, value=1}
          {arriveTangent=-4.4288, interpMode=RCIM_Cubic, leaveTangent=-4.4288, tangentMode=RCTM_User, time=1, value=0}
         ]}}}
      Scale Curve=2
     )
    )
    Scale RGB=&MaskVectorBySpawnGroup(
     Masked Vector=&VectorFromCurve(VectorCurve=DataInterface() {propertyValues {
        XCurve {keys [<at 0.25 1 RCIM_Cubic>,<at 1 0.5>]}
        YCurve {keys [<at 0.25 1 RCIM_Cubic>,<at 1 0.5>]}
        ZCurve {keys [<at 0.25 1 RCIM_Cubic>,<at 1 0.5>]}
       }})
     Passthrough Vector=<0 0 0>
    )
    ScaleA=true
    ScaleRGB=true
   )
   /Niagara/Modules/Update/SubUV/SubUVAnimation(Number Of Frames=32)
   &DynamicMaterialParameters(
    Index 0 Param 1=&MaskFloatBySpawnGroup(
     Masked Float=&FloatFromCurve(
      FloatCurve=DataInterface() {propertyValues {Curve {keys [
          {arriveTangent=2.8718, interpMode=RCIM_Cubic, leaveTangent=2.8718, tangentMode=RCTM_User, time=0, value=0}
          {arriveTangent=-0.3942, interpMode=RCIM_Cubic, leaveTangent=-0.3942, tangentMode=RCTM_User, time=1, value=1}
         ]}}}
      Scale Curve=0.1
     )
     Passthrough Float=0
    )
    Index 0 Param 2=80
    Module.Index0Param0Write=true
    Module.Index0Param1Write=true
    Module.Index0Param2Write=false
    Module.Index0Param3Write=false
   ) {module=DynamicMaterialParameters001}
   &Drag(Use Linear Drag=true, Use Rotational Drag=true) {enabled=false}
   &SolveForcesAndVelocity(Clamp Velocity=false, Limit Acceleration=false)
  ]
  rendererClasses [&NiagaraSpriteRendererProperties]
  renderers [{rendererClass=&NiagaraSpriteRendererProperties}]
 }
 SmokeLoop_LW {rendererClasses [], renderers []}
 Trail {
  bEnabled=false
  emitterUpdateScript [
   &EmitterState(Life Cycle Mode=::NewEnumerator1, Loop Behavior=::NewEnumerator1, Loop Delay=0.15, Loop Duration=1.2, UseLoopDelay=true)
   &SpawnPerUnit(Module.Use Max Movement Threshold=false, Module.Use Movement Tolerance=false, Use Spawn Probability=false) {enabled=false, module=SpawnPerUnit001}
   &SpawnPerUnit(
    Module.Use Max Movement Threshold=false
    Module.Use Movement Tolerance=false
    Movement Tolerance=20
    Spawn Spacing:NiagaraFloat=&Lerp_Float(
     A=20
     B=100
     Alpha=&FloatFromCurve(
      FloatCurve=DataInterface() {propertyValues {Curve {keys [
          {interpMode=RCIM_Cubic, tangentMode=RCTM_User, time=0.1, value=0}
          {arriveTangent=5.1366, interpMode=RCIM_Cubic, leaveTangent=5.1366, tangentMode=RCTM_User, time=0.5, value=1}
         ]}}}
      CurveIndex=$Emitter.Age
     )
    )
    Use Spawn Probability=false
   )
   &SpawnRate(SpawnRate=3, Use Spawn Probability=false) {enabled=false}
  ]
  particleSpawnScript [
   &InitializeParticle(
    Lifetime=1.5
    Mass Mode=::NewEnumerator3
    Ribbon Width=400
    Ribbon Width Mode=::NewEnumerator1
    Sprite Rotation Mode=::NewEnumerator2
    Sprite Size Mode=::NewEnumerator3
    Uniform Sprite Size=50
    UsePositionOffset=false
    Write Material Random=true
   )
   &SkeletalMeshLocation(
    Bone Sampling Mode=::NewEnumerator4
    MaskSpawnGroup=false
    Skeletal Mesh=DataInterface() {propertyValues {
      DefaultMesh=null
      ExcludeBoneName=root
      OptionalComponentClass=null
      PreviewMesh=&SK_M_MALE_Base_Skeleton
      SkinningMode=SkinOnTheFly
      SoftSourceActor=null
      SourceMode=Default
      UvSetIndex=0
      WholeMeshLOD=-1
      bExcludeBone=false
      bReadDeformedGeometry=true
      bRequireCurrentFrameData=true
      ComponentTags []
      FilteredBones [pelvis]
      FilteredSockets []
      MeshUserParameter {parameter {name=None, typeDefHandle {registeredTypeIndex=1}, varData []}}
      SamplingRegions []
     }}
   )
   &AddVelocity(Velocity=<0 0 20>)
   &InheritVelocity(Inherited Velocity Speed Limit=200, Speed Limit=true, UseSpeedThreshold=false)
  ]
  particleUpdateScript [
   &ParticleState()
   &ScaleColor(
    Scale Alpha=&FloatFromCurve(
     FloatCurve=DataInterface() {propertyValues {Curve {keys [<at 0 0 RCIM_Cubic>,<at 0.13 1 RCIM_Cubic>,<at 0.8 0 RCIM_Cubic>]}}}
     Scale Curve=0.8
    )
    Scale RGB=&MakeVectorFromLinearColorRGB(Color=&ColorFromCurve(ColorCurve=DataInterface() {propertyValues {
        AlphaCurve {keys [<at 0 1 RCIM_Cubic>]}
        BlueCurve {keys [<at 0.0141 2 RCIM_Cubic>,<at 0.0609 1>,<at 0.1734 0.3>,<at 0.7509 0 RCIM_Cubic>]}
        GreenCurve {keys [<at 0.0141 0.2475 RCIM_Cubic>,<at 0.0609 0.0955>,<at 0.1734 0.03>,<at 0.7509 0 RCIM_Cubic>]}
        RedCurve {keys [<at 0.0141 1.9531 RCIM_Cubic>,<at 0.0609 0.6675>,<at 0.1734 0.075>,<at 0.7509 0 RCIM_Cubic>]}
       }}))
    ScaleA=true
    ScaleRGB=true
   )
   &CurlNoiseForce(
    Bias Noise=false
    Noise Frequency=2
    Noise Strength=&Multiply_Float(
     A=220
     B=&FloatFromCurve(FloatCurve=DataInterface() {propertyValues {Curve {keys [{interpMode=RCIM_Cubic, tangentMode=RCTM_User, time=0, value=1}, {interpMode=RCIM_Cubic, tangentMode=RCTM_User, time=0.7, value=0}]}}})
    )
    Pan Noise=false
   )
   &Drag(Drag=4, Use Linear Drag=true, Use Rotational Drag=true)
   &SolveForcesAndVelocity(Clamp Velocity=false, Limit Acceleration=false)
   &DynamicMaterialParameters(
    Index 0 Param 1=&FloatFromCurve(FloatCurve=DataInterface() {propertyValues {Curve {keys [
         {arriveTangent=-1.2682, interpMode=RCIM_Cubic, leaveTangent=-1.2682, tangentMode=RCTM_User, time=0.1, value=1}
         {interpMode=RCIM_Cubic, tangentMode=RCTM_User, time=0.8, value=-0.2}
        ]}}})
    Index 0 Param 2=1
    Index 0 Param 3=0.2
    Module.Index0Param0Write=true
    Module.Index0Param1Write=true
    Module.Index0Param2Write=true
    Module.Index0Param3Write=false
   )
   &ScaleRibbonWidth(Ribbon Width Scale=&FloatFromCurve(
     FloatCurve=DataInterface() {propertyValues {Curve {keys [
         {arriveTangent=1.8307, interpMode=RCIM_Cubic, leaveTangent=1.8307, tangentMode=RCTM_User, time=0, value=0.2}
         {arriveTangent=0.3065, interpMode=RCIM_Cubic, leaveTangent=0.3065, tangentMode=RCTM_User, time=0.75, value=1}
        ]}}}
     Scale Curve=$System.FX Scale
    ))
  ]
  rendererClasses [&NiagaraRibbonRendererProperties]
  renderers [{rendererClass=&NiagaraRibbonRendererProperties}]
 }
 TwistAbsorb {
  emitterUpdateScript [
   &EmitterState(Life Cycle Mode=::NewEnumerator1, Loop Behavior=::NewEnumerator1, Loop Delay=0.18, Loop Duration=2, UseLoopDelay=true)
   &SpawnBurst_Instantaneous(Use Spawn Probability=false, UseLoopCountLimit=false)
  ]
  particleSpawnScript [
   &InitializeParticle(
    Color=&MakeCustomLinearColorFromBool(Bool=$User.WanderingBoss, True Color=<rgba 6 0.24 0.816 1>, False Color=<rgba 1.872 0.24 6 1>)
    Lifetime=0.33
    Position Offset=$System.HandFXOffset
    Sprite Rotation Mode=::NewEnumerator2
    Sprite Size Mode=::NewEnumerator3
    Uniform Sprite Size=140
    UsePositionOffset=false
    Write Material Random=true
   )
   &CameraOffset(Camera Offset Amount=200) {module=CameraOffset001}
  ]
  particleUpdateScript [
   &CameraOffset(Camera Offset Amount=&Lerp_Float(
     A=600
     B=-100
     Alpha=&FloatFromCurve(FloatCurve=DataInterface() {propertyValues {Curve {keys [{interpMode=RCIM_Cubic, tangentMode=RCTM_User, time=0.35, value=0}, {interpMode=RCIM_Cubic, tangentMode=RCTM_User, time=0.7, value=1}]}}})
    )) {enabled=false}
   &ParticleState()
   &SkeletalMeshLocation(
    Bone Sampling Mode=::NewEnumerator4
    MaskSpawnGroup=false
    Orientation Sampling=::NewEnumerator1
    Sampled Position Offset=<50 0 0>
    Skeletal Mesh=DataInterface() {propertyValues {
      DefaultMesh=null
      ExcludeBoneName=root
      OptionalComponentClass=null
      PreviewMesh=&SK_M_MALE_Base_Skeleton
      SkinningMode=SkinOnTheFly
      SoftSourceActor=null
      SourceMode=Default
      UvSetIndex=0
      WholeMeshLOD=-1
      bExcludeBone=false
      bReadDeformedGeometry=true
      bRequireCurrentFrameData=true
      ComponentTags []
      FilteredBones [pelvis]
      FilteredSockets []
      MeshUserParameter {parameter {name=None, typeDefHandle {registeredTypeIndex=1}, varData []}}
      SamplingRegions []
     }}
   )
   &ScaleSpriteSize(
    Uniform Curve Scale=$System.FX Scale
    Uniform Curve Sprite Scale=DataInterface() {propertyValues {Curve {keys [
        {arriveTangent=12.5362, interpMode=RCIM_Cubic, leaveTangent=12.5362, tangentMode=RCTM_User, time=0, value=0.3}
        {arriveTangent=0.2602, interpMode=RCIM_Cubic, leaveTangent=0.2602, tangentMode=RCTM_User, time=0.1, value=1}
       ]}}}
   )
   &SolveForcesAndVelocity(Clamp Velocity=false, Limit Acceleration=false)
   &DynamicMaterialParameters(
    Index 0 Param 1=&Lerp_Float(
     A=-0.2
     B=1
     Alpha=&FloatFromCurve(FloatCurve=DataInterface() {propertyValues {Curve {keys [
          {arriveTangent=0.1556, interpMode=RCIM_Cubic, leaveTangent=0.1556, tangentMode=RCTM_User, time=0, value=0}
          {arriveTangent=2.2716, interpMode=RCIM_Cubic, leaveTangent=2.2716, tangentMode=RCTM_User, time=0.8, value=1}
         ]}}})
    )
    Index 0 Param 2=1
    Index 0 Param 3=&Lerp_Float(
     A=-0.1
     B=1
     Alpha=&FloatFromCurve(FloatCurve=DataInterface() {propertyValues {Curve {keys [
          {arriveTangent=0.1235, interpMode=RCIM_Cubic, leaveTangent=0.1235, tangentMode=RCTM_User, time=0.2, value=0}
          {arriveTangent=1.955, interpMode=RCIM_Cubic, leaveTangent=1.955, tangentMode=RCTM_User, time=1, value=1}
         ]}}})
    )
    Index 0 Param 4=&Lerp_Float(
     A=0
     B=0.25
     Alpha=&FloatFromCurve(FloatCurve=DataInterface() {propertyValues {Curve {keys [
          {arriveTangent=0.3497, interpMode=RCIM_Cubic, leaveTangent=0.3497, tangentMode=RCTM_User, time=0, value=0}
          {interpMode=RCIM_Cubic, tangentMode=RCTM_User, time=1, value=1}
         ]}}})
    )
    Module.Index0Param0Write=true
    Module.Index0Param1Write=true
    Module.Index0Param2Write=true
    Module.Index0Param3Write=true
   )
   &ScaleColor(
    Scale Alpha=&FloatFromCurve(
     FloatCurve=DataInterface() {propertyValues {Curve {keys [
         <at 0 0 RCIM_Cubic>
         {arriveTangent=-0.0178, interpMode=RCIM_Cubic, leaveTangent=-0.0178, tangentMode=RCTM_User, time=0.1, value=1}
         {interpMode=RCIM_Cubic, tangentMode=RCTM_User, time=0.3, value=1}
         {interpMode=RCIM_Cubic, tangentMode=RCTM_User, time=1, value=0}
        ]}}}
     Scale Curve=2
    )
    Scale RGB=&MakeVectorFromLinearColorRGB(Color=&ColorFromCurve(ColorCurve=DataInterface() {propertyValues {
        AlphaCurve {keys [<at 0 1 RCIM_Cubic>]}
        BlueCurve {keys [<at 0.1216 1 RCIM_Cubic>,<at 0.8838 0.03 RCIM_Cubic>]}
        GreenCurve {keys [<at 0.1216 1 RCIM_Cubic>,<at 0.8838 0.0018 RCIM_Cubic>]}
        RedCurve {keys [<at 0.1216 1 RCIM_Cubic>,<at 0.8838 0.0065 RCIM_Cubic>]}
       }}))
    ScaleA=true
    ScaleRGB=true
   )
   /SetVariables(
    Particles.ColorDark:LinearColor=&MultiplyLinearColor(
     Linear Color=&MakeCustomLinearColorFromBool(Bool=true, True Color=<rgba 0.045 0 0.006 1>, False Color=<rgba 0.0014 0 0.045 1>)
     Scale Factor=&MakeFloat4FromVectorAndFloat(
      "Vector (RGB)"=<1 1 1>
      "Float (Alpha)"=&FloatFromCurve(
       FloatCurve=DataInterface() {propertyValues {Curve {keys [
           <at 0 0 RCIM_Cubic>
           {arriveTangent=-2.1159, interpMode=RCIM_Cubic, leaveTangent=-2.1159, tangentMode=RCTM_User, time=0.2, value=1}
           {interpMode=RCIM_Cubic, tangentMode=RCTM_User, time=1, value=0}
          ]}}}
       Scale Curve=1.6
      )
     )
    )
    Particles.DynamicMaterialParameterDark:Vector4f=&MakeVector4(
     X=&Lerp_Float(
      A=0
      B=1
      Alpha=&FloatFromCurve(FloatCurve=DataInterface() {propertyValues {Curve {keys [
           {interpMode=RCIM_Cubic, tangentMode=RCTM_User, time=0, value=0}
           {arriveTangent=1.7687, interpMode=RCIM_Cubic, leaveTangent=1.7687, tangentMode=RCTM_User, time=1, value=1}
          ]}}})
     )
     Y=1
     Z=&Lerp_Float(
      A=-0.5
      B=1
      Alpha=&FloatFromCurve(FloatCurve=DataInterface() {propertyValues {Curve {keys [
           {arriveTangent=0.1235, interpMode=RCIM_Cubic, leaveTangent=0.1235, tangentMode=RCTM_User, time=0.2, value=0}
           {arriveTangent=1.955, interpMode=RCIM_Cubic, leaveTangent=1.955, tangentMode=RCTM_User, time=1, value=1}
          ]}}})
     )
     W=0.232
    )
    Particles.SpriteSizeDark:Vector2f=&Multiply_Vector2D_ByFloat(Float=1.4, Vector2D=$Particles.SpriteSize)
   ) {isSetParameters=true}
  ]
  rendererClasses [&NiagaraSpriteRendererProperties]
  renderers [
   {rendererClass=&NiagaraSpriteRendererProperties}
   {rendererClass=&NiagaraSpriteRendererProperties, rendererIndex=1}
  ]
 }
}
summary {
 systemName=NS_GrillMoonTeleportStart
 emitters [
  {emitterName=TwistAbsorb, rendererClasses [&NiagaraSpriteRendererProperties]}
  {bEnabled=false, emitterName=BodyDecals, rendererClasses [&NiagaraSpriteRendererProperties]}
  {bEnabled=false, emitterName=Trail, rendererClasses [&NiagaraRibbonRendererProperties]}
  {emitterName=ExitPortal, rendererClasses [&NiagaraMeshRendererProperties, &NiagaraSpriteRendererProperties]}
  {emitterName=Copied_ChromaticRefraction, rendererClasses [&NiagaraSpriteRendererProperties]}
  {emitterName=SmokeLoop_LW, rendererClasses []}
  {emitterName=SmokeLoop, rendererClasses [&NiagaraSpriteRendererProperties]}
 ]
 userVariables [
  User.Global Scale {type=NiagaraFloat, defaultValue=1, description=""}
  User.GlobalScale {type=NiagaraFloat, defaultValue=1, description=""}
  User.Offset Pos {type=NiagaraPosition, description="", defaultValue {}}
  User.WanderingBoss {type=NiagaraBool, defaultValue=true}
 ]
}
